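# Skydiving contrail cosmetic that trails flares, glows, sparks and lightning meshes from the player's skeletal mesh, using spawn group as bone index.
# Normalized system speed drives lightning spawn rates, while a bone-derived dive direction stretches and scales particles.
# All emitters sample the user HueRamp via a system float integrator, mostly phase-offset, to cycle color.
# Front-end modifiers substitute ParaGlideLeanAlpha or fixed values for gameplay motion.
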NiagaraSystem /BRCosmetics/Athena/Effects/Particles/Skydiving/NS_Contrail_ReactiveLight/NS_Contrail_ReactiveLight
dependencies {
 usedDataInterfaces [&NiagaraDataInterfaceCurve, &NiagaraDataInterfaceVector2DCurve]
 usedDynamicInputs [&SimulationPosition]
 usedModules [
  &SystemState
  /Game/Effects/Niagara/Modules/Skydiving/FN_NormalizedSystemVel
  /Game/Effects/Niagara/Modules/Math/FN_SystemFloatIntegrator
 ]
}
emitters {
 2_Flare {
  emitterUpdateScript [
   &EmitterState()
   &SpawnRate(SpawnRate=60, Use Spawn Probability=false)
  ]
  particleSpawnScript [
   &FN_SetSpawnGroup() {module=FN_SetSpawnGroup001}
   /SetVariables(Particles.Component:Vector3f=&MakeVector(
     X=&MakeFloatFromInt(Int=&UniformRangedInt(Minimum=0, Maximum=2))
     Y=&MakeFloatFromInt(Int=&UniformRangedInt(Minimum=0, Maximum=2))
     Z=&MakeFloatFromInt(Int=&UniformRangedInt(Minimum=0, Maximum=2))
    )) {isSetParameters=true}
   &InitializeParticle(
    Color=<rgba 1 0 0 1>
    Lifetime=&UniformRangedFloat(Minimum=0.4, Maximum=0.5)
    Sprite Size=&Multiply_Vector2(
     A=&UniformRangedVector2D(Minimum=<10 25>, Maximum=<15 50>)
     B=&Vector2DFromFloat(Value=HlslExpression() {hlslExpression="Particles.SpawnGroup < 2 ?  1 : .75"})
    )
    Sprite Size Mode=::NewEnumerator4
    UsePositionOffset=false
    Write Material Random=true
   ) {module=InitializeParticle001}
  ]
  particleUpdateScript [
   /Game/Effects/Niagara/Modules/SkeletalMesh/FN_SkeletalBoneData(Index=4)
   /SetVariables(Particles.DiveDirection:NiagaraFloat=&ClampFloat(Float=&Multiply_Float(
      A=&FN_FrontEndFloatModifier(Gameplay=1.25, Front End=2)
      B=&DotProduct(
       Vector A=&RotateVectorByQuaternion(
        Vector To Rotate=<1 0 0>
        Quaternion=/Game/Athena/DrivableVehicles/Meatball/Meatball_Large/FX/DynamicInputs/FN_GetBoneDataRotation(BoneData:ENiagaraBoneData=$Particles.FN_SkeletalBoneData.BoneData, Space:ENiagaraCoordinateSpace=::World)
       )
       Vector B=<0 0 -1>
      )
     ))) {isSetParameters=true}
   &ParticleState()
   &SampleSkeletalMeshSkeleton(
    Interpolate Bone Sampling=false
    Mesh:NiagaraDataInterfaceSkeletalMesh=$User.SkeletalMesh
    Specific Bone Index:NiagaraInt32=$Particles.SpawnGroup
   )
   &SkeletalMeshLocation(Bone / Socket Index=$Particles.SpawnGroup, Bone Sampling Mode=::NewEnumerator4, MaskSpawnGroup=false, Skeletal Mesh=$User.SkeletalMesh)
   /Game/Effects/Niagara/Modules/Materials/MaterialComponentMask(
    Blue=0
    Green:NiagaraFloat=&MakeFloatFromVector(Vector=$Particles.Component, Channel=::NewEnumerator1)
    Red:NiagaraFloat=&MakeFloatFromVector(Vector=$Particles.Component, Channel=::NewEnumerator0)
   )
   &Material(
    Addition=0.0727
    Base Texture Time=HlslExpression() {hlslExpression=Emitter.Age + Particles.SpawnGroup*3.9+System.FloatIntegrator*.3}
    Distortion Texture Time:NiagaraFloat=$Emitter.Age
    Power=1.5416
   )
   &Color(
    Color=&ColorFromCurve(ColorCurve=$User.HueRamp, CurveIndex=HlslExpression() {hlslExpression="fmod((System.FloatIntegrator)-.16666,1.0)"})
    Scale Alpha=&FloatFromCurve(
     FloatCurve=DataInterface() {propertyValues {Curve {keys [<at -0.0045 0.0315>,<at 0.0491 0.9978>,<at 0.942 0.9978>,<at 1.0011 0.002>]}}}
     Scale Curve=&FloatFromCurve(FloatCurve=DataInterface() {propertyValues {Curve {keys [<at 0.5 1>,<at 1.0006 0>]}}}, CurveIndex=$Emitter.Scalability.DistanceFraction)
    )
    Scale Color=<0.5 0.5 0.5>
   )
   &ScaleSpriteSize(Scale Factor=&Vector2DFromCurve(
     Vector2Curve=DataInterface() {propertyValues {XCurve {keys [<at -0.0011 1.0067>,<at 1.0149 2.7937>]}, YCurve {keys [<at -0.0011 1.0067>,<at 0.996 3.9474>]}}}
     CurveIndex=&FN_FrontEndFloatModifier(Gameplay=$Particles.DiveDirection, Front End=$User.ParaGlideLeanAlpha)
     ScaleCurve=<1 1>
    ))
   &CameraOffset(Camera Offset Amount=&Add_Float(A=20, B=$Particles.RandomizedFloat))
   /SetVariables(Particles.SpriteAlignment=/Niagara/DynamicInputs/Helpers/SafeNormalizeVector(
     Vector To Normalize:Vector3f=&Multiply_VectorByFloat(
      Vector=&FN_FrontEndVector3Modifier(
       Gameplay=&Multiply_Vector(
        A=&SafeNormalizeVector(Vector To Normalize:Vector3f=$Particles.SampleSkeletalMeshSkeleton.SampledVelocity, Fallback Vector:Vector3f=<0 0 -1>)
        B=&VectorFromFloat(Value=/Niagara/DynamicInputs/TypeConversions/VectorLengthSafe(Vector:Vector3f=$Particles.SampleSkeletalMeshSkeleton.SampledVelocity))
       )
       Front End=&Lerp_Vector(
        A=<0 0 -1>
        B=&TransformVector(Vector=<0 1.5 -1>, Destination Space=::World)
        Alpha=$User.ParaGlideLeanAlpha
       )
      )
      Float=-1
     )
     Fallback Vector:Vector3f=<0 0 1>
    )) {isSetParameters=true}
  ]
  rendererClasses [&NiagaraSpriteRendererProperties]
  renderers [{rendererClass=&NiagaraSpriteRendererProperties}]
 }
 3_Glow {
  emitterUpdateScript [
   &EmitterState()
   &SpawnRate(SpawnRate=20, Use Spawn Probability=false)
  ]
  particleSpawnScript [
   &FN_SetSpawnGroup() {module=FN_SetSpawnGroup001}
   &InitializeParticle(
    Color=<rgba 1 0 0 1>
    Lifetime=&UniformRangedFloat(Minimum=0.5, Maximum=1)
    Sprite Rotation Angle=&UniformRangedFloat(Minimum=0, Maximum=360)
    Sprite Rotation Mode=::NewEnumerator0
    Sprite Size=&UniformRangedVector2D(Minimum=<40 100>, Maximum=<100 100>)
    Sprite Size Mode=::NewEnumerator4
    UsePositionOffset=false
    Write Material Random=true
   ) {module=InitializeParticle001}
   /SetVariables(Particles.SpriteSize=&Multiply_Vector2(
     A=$Particles.SpriteSize
     B=&Vector2DFromFloat(Value=&FN_FrontEndFloatModifier(Gameplay=1, Front End=2))
    )) {isSetParameters=true}
  ]
  particleUpdateScript [
   &FN_SkeletalBoneData(Index=4)
   /SetVariables(Particles.DiveDirection:NiagaraFloat=&ClampFloat(Float=&Multiply_Float(
      A=&FN_FrontEndFloatModifier(Gameplay=1.25, Front End=2)
      B=&DotProduct(
       Vector A=&RotateVectorByQuaternion(
        Vector To Rotate=<1 0 0>
        Quaternion=&FN_GetBoneDataRotation(BoneData:ENiagaraBoneData=$Particles.FN_SkeletalBoneData.BoneData, Space:ENiagaraCoordinateSpace=::World)
       )
       Vector B=<0 0 -1>
      )
     ))) {isSetParameters=true}
   &ParticleState()
   &SampleSkeletalMeshSkeleton(Mesh:NiagaraDataInterfaceSkeletalMesh=$User.SkeletalMesh, Specific Bone Index:NiagaraInt32=$Particles.SpawnGroup)
   /SetVariables(Particles.Position=$Particles.SampleSkeletalMeshSkeleton.SampledPosition) {isSetParameters=true}
   /Game/Effects/Niagara/Modules/Materials/MaterialUV()
   &MaterialComponentMask(
    Blue=0
    Green:NiagaraFloat=&MakeFloatFromVector(Vector=$Particles.Component, Channel=::NewEnumerator1)
    Red:NiagaraFloat=&MakeFloatFromVector(Vector=$Particles.Component, Channel=::NewEnumerator0)
   )
   &Material(
    Addition=0.0727
    Base Texture Time=HlslExpression() {hlslExpression=Emitter.Age + Particles.SpawnGroup*3.9}
    Distortion Texture Time:NiagaraFloat=$Emitter.Age
    Power=1.5416
   )
   &Color(
    Color=&ColorFromCurve(ColorCurve=$User.HueRamp, CurveIndex=HlslExpression() {hlslExpression="fmod((System.FloatIntegrator),1.0)"})
    Scale Alpha=&FloatFromCurve(
     FloatCurve=DataInterface() {propertyValues {Curve {keys [<at -0.0045 0.0315>,<at 0.0491 0.9978>,<at 0.942 0.9978>,<at 1.0011 0.002>]}}}
     Scale Curve=0.2
    )
   )
   &ScaleSpriteSize(Scale Factor=&Vector2DFromFloat(Value=&FloatFromCurve(
      FloatCurve=DataInterface() {propertyValues {Curve {keys [<at -0.0275 0.0349>,<at 0.4537 1.1551>,<at 0.5879 1.1222>,<at 1.0194 0.0349>]}}}
      Scale Curve=&FloatFromCurve(
       FloatCurve=DataInterface() {propertyValues {Curve {keys [<at -0.002 0.8595>,<at 1.0073 1.7699>]}}}
       CurveIndex=&FN_FrontEndFloatModifier(Gameplay=$Particles.DiveDirection, Front End=$User.ParaGlideLeanAlpha)
       Scale Curve=1.0986
      )
     )))
   &CameraOffset(Camera Offset Amount=&Add_Float(A=-30, B=$Particles.RandomizedFloat))
   /SetVariables(
    Particles.CameraOffset2:NiagaraFloat=&Add_Float(A=$Particles.CameraOffset, B=40)
    Particles.Color2=&MultiplyLinearColor(Linear Color=$Particles.Color, Scale Factor=&MakeVector4_FromFloat(Float=0.5))
   ) {isSetParameters=true}
   &SolveForcesAndVelocity(Clamp Velocity=false, Limit Acceleration=false)
  ]
  rendererClasses [&NiagaraSpriteRendererProperties]
  renderers [{rendererClass=&NiagaraSpriteRendererProperties}]
 }
 4_FN_Sparks {
  emitterUpdateScript [
   &EmitterState()
   &SpawnRate(
    SpawnRate=&FloatFromCurve(
     FloatCurve=DataInterface() {propertyValues {Curve {keys [<at -0.0461 3.276>,<at 1.0002 3.3506>]}}}
     CurveIndex=$System.NormalizedSystemSpeed
     Scale Curve=60
    )
    Use Spawn Probability=true
   )
  ]
  particleSpawnScript [
   &FN_SetSpawnGroup() {module=FN_SetSpawnGroup001}
   &InitializeParticle(
    Color=<rgba 1 0 0 1>
    Lifetime=&UniformRangedFloat(Minimum=0.2, Maximum=0.5)
    Sprite Rotation Angle=&UniformRangedFloat(Minimum=0, Maximum=360)
    Sprite Rotation Mode=::NewEnumerator0
    Sprite Size=&Multiply_Vector2(
     A=&UniformRangedVector2D(Minimum=<2 2>, Maximum=<3 10>)
     B=&Vector2DFromFloat(Value=HlslExpression() {hlslExpression="Particles.SpawnGroup < 2 ?  1 : .5"})
    )
    Sprite Size Mode=::NewEnumerator4
    UsePositionOffset=false
    Write Material Random=true
   ) {module=InitializeParticle001}
   &SampleSkeletalMeshSkeleton(
    Mesh:NiagaraDataInterfaceSkeletalMesh=$User.SkeletalMesh
    Sampled Position Coordinate Space:ENiagaraCoordinateSpace=::Local
    Sampled Velocity Coordinate Space:ENiagaraCoordinateSpace=::World
    Specific Bone Index:NiagaraInt32=$Particles.SpawnGroup
   )
   &SkeletalMeshLocation(MaskSpawnGroup=false, OverrideRandomSeed=false, Skeletal Mesh=$User.SkeletalMesh)
   &FN_SkeletalBoneData(Index=4)
   /SetVariables(Particles.DiveDirection:NiagaraFloat=&ClampFloat(Float=&Multiply_Float(
      A=1.25
      B=&DotProduct(
       Vector A=&RotateVectorByQuaternion(
        Vector To Rotate=<1 0 0>
        Quaternion=&FN_GetBoneDataRotation(BoneData:ENiagaraBoneData=$Particles.FN_SkeletalBoneData.BoneData, Space:ENiagaraCoordinateSpace=::Local)
       )
       Vector B=&TransformVector(Vector=<0 0 -1>, Source Space=::World, Destination Space=::Local)
      )
     ))) {isSetParameters=true}
   /Niagara/Modules/Spawn/Location/SphereLocation(
    MaskSpawnGroup=false
    Module.Override Randomness=false
    Sphere Origin:NiagaraPosition=$Particles.SampleSkeletalMeshSkeleton.SampledPosition
    Sphere Radius=5
    Surface Only=false
   )
   /SetVariables(Particles.Blip:NiagaraFloat=/Game/Effects/Niagara/DynamicInputs/Time/FN_Blip(LoopCount:NiagaraInt32=$Emitter.LoopCount, Probability:NiagaraFloat=0.5, Seed:NiagaraInt32=0, SpawnGroup:NiagaraInt32=$Particles.SpawnGroup)) {isSetParameters=true}
   /SetVariables(Particles.SpriteSize=/Niagara/DynamicInputs/Add/Add_Vector2(
     A:Vector2f=$Particles.SpriteSize
     B:Vector2f=&Vector2DFromFloat(Value=&Multiply_Float(A=$Particles.Blip, B=HlslExpression() {hlslExpression="Particles.SpawnGroup < 2 ?  5 : 1"}))
    )) {isSetParameters=true}
   /SetVariables(Particles.SpriteSize=&Multiply_Vector2(
     A=$Particles.SpriteSize
     B=&MakeVector2D(
      X=1
      Y=&FloatFromCurve(
       FloatCurve=DataInterface() {propertyValues {Curve {keys [<at -0.03 0.4>,<at 0.9874 4>]}}}
       CurveIndex=&FN_FrontEndFloatModifier(Gameplay=$Particles.DiveDirection, Front End=$User.ParaGlideLeanAlpha)
       Scale Curve=3
      )
     )
    )) {isSetParameters=true}
   /SetVariables(Particles.Velocity=&Multiply_VectorByFloat(
     Vector=&FN_FrontEndVector3Modifier(
      Gameplay=&SafeNormalizeVector(
       Vector To Normalize:Vector3f=&Multiply_VectorByFloat(
        Vector=&TransformVector(Vector=$Particles.SampleSkeletalMeshSkeleton.SampledVelocity, Source Space=::World, Destination Space=::Local)
        Float=1
       )
       Fallback Vector:Vector3f=<0 0 1>
      )
      Front End=/Game/Effects/Niagara/DynamicInputs/FrontEnd/FN_FrontEndRotateVector(Angle:NiagaraFloat=$User.ParaGlideLeanAlpha, Vector to Rotate:Vector3f=<0 0 -1>)
     )
     Float=&Lerp_Float(
      A=&Add_Float(A=-75, B=&Multiply_Float(A=$Particles.Blip, B=-500))
      B=-1000
      Alpha=&FloatFromCurve(
       FloatCurve=DataInterface() {propertyValues {Curve {keys [<at 0.25 0>,<at 0.75 1>]}}}
       CurveIndex=&Add_Float(A=$System.NormalizedLoopAge, B=$User.ParaGlideLeanAlpha)
      )
     )
    )) {isSetParameters=true}
   /SetVariables(Particles.Lifetime=&Add_Float(A=$Particles.Lifetime, B=$Particles.Blip)) {isSetParameters=true}
  ]
  particleUpdateScript [
   &ParticleState()
   /Niagara/Modules/Update/Forces/CurlNoiseForce(
    Noise Quality / Cost=::NewEnumerator0
    Noise Strength=&Lerp_Float(
     A=500
     B=1000
     Alpha=&FN_FrontEndFloatModifier(Gameplay=$Particles.DiveDirection, Front End=$User.ParaGlideLeanAlpha)
    )
    Pan Noise=true
    Pan Noise Field=<0 0 1>
   )
   &SolveForcesAndVelocity(Clamp Velocity=false, Limit Acceleration=false)
   &Color(
    Color=&ColorFromCurve(ColorCurve=$User.HueRamp, CurveIndex=HlslExpression() {hlslExpression="fmod(System.FloatIntegrator-.166+(Particles.NormalizedAge*.2),1.0)"})
    Scale Alpha=&FloatFromCurve(FloatCurve=DataInterface() {propertyValues {Curve {keys [<at -0.0046 1.0113>,<at 1.0011 0.002>]}}})
    Scale Color=<2 2 2>
   )
   &ScaleSpriteSize(Scale Factor=&Vector2DFromCurve(
     Vector2Curve=DataInterface() {propertyValues {XCurve {keys [<at -0.0017 1>,<at 0.9933 0.7941>]}, YCurve {keys [<at -0.0011 1.0067>,<at 0.9989 0>]}}}
     ScaleCurve=<1 1>
    ))
  ]
  rendererClasses [&NiagaraSpriteRendererProperties]
  renderers [{rendererClass=&NiagaraSpriteRendererProperties}]
 }
 4_Lightning {
  emitterUpdateScript [
   &EmitterState()
   &SpawnRate(
    SpawnRate=&FloatFromCurve(FloatCurve=DataInterface() {propertyValues {Curve {keys [<at 0.5488 1.0003>,<at 1.0081 4.1405>]}}}, CurveIndex=$System.NormalizedSystemSpeed)
    Use Spawn Probability=false
   )
  ]
  particleSpawnScript [
   /SetVariables(Particles.SpawnGroup=HlslExpression() {hlslExpression="int(fmod(float(Particles.UniqueID), 4.0))"}) {isSetParameters=true, module=SetVariables_#1}
   &InitializeParticle(
    Lifetime=&UniformRangedFloat(Minimum=0.1, Maximum=0.2)
    Mesh Scale=&UniformRangedVector(Minimum=<0.2 0.1 0.1>, Maximum=<0.7 0.1 0.1>)
    Mesh Scale Mode=::NewEnumerator4
    UsePositionOffset=false
    Write Material Random=true
   ) {module=InitializeParticle001}
  ]
  particleUpdateScript [
   &FN_SkeletalBoneData(Index=4)
   /SetVariables(Particles.DiveDirection:NiagaraFloat=&ClampFloat(Float=&Multiply_Float(
      A=&FN_FrontEndFloatModifier(Gameplay=1.25, Front End=2)
      B=&DotProduct(
       Vector A=&RotateVectorByQuaternion(
        Vector To Rotate=<1 0 0>
        Quaternion=&FN_GetBoneDataRotation(BoneData:ENiagaraBoneData=$Particles.FN_SkeletalBoneData.BoneData, Space:ENiagaraCoordinateSpace=::World)
       )
       Vector B=<0 0 -1>
      )
     ))) {isSetParameters=true}
   &ParticleState()
   &SampleSkeletalMeshSkeleton(Mesh:NiagaraDataInterfaceSkeletalMesh=$User.SkeletalMesh, Specific Bone Index:NiagaraInt32=$Particles.SpawnGroup)
   &ScaleMeshSize(Scale Factor=&Lerp_Vector(A=<1 2 2>, B=<2 4 4>, Alpha=$Particles.DiveDirection))
   /BRCosmetics/Athena/Effects/Particles/Skydiving/NS_Contrail_ReactiveLight/Modules/FN_OrientAndSpin(Axis:Vector3f=&FN_FrontEndVector3Modifier(
     Gameplay=$Particles.SampleSkeletalMeshSkeleton.SampledVelocity
     Front End=&Lerp_Vector(
      A=<0 0 -1>
      B=&TransformVector(Vector=<0 1.2 -1>, Source Space=::World, Destination Space=::Local)
      Alpha=$User.ParaGlideLeanAlpha
     )
    ))
   /SetVariables(Particles.Position=$Particles.SampleSkeletalMeshSkeleton.SampledPosition) {isSetParameters=true}
   &DynamicMaterialParameters(
    Index 0 Param 1=&FloatFromCurve(FloatCurve=DataInterface() {propertyValues {Curve {keys [<at -0.0046 1.0113>,<at 1.0005 0.0072>]}}})
    Index 0 Param 2=0.4028
    Index 0 Param 3=0.005
    Index 0 Param 4=0.1314
    Module.Param0WriteEnabled=true
    Module.Param1WriteEnabled=true
    Module.Param2WriteEnabled=true
    Module.Param3WriteEnabled=true
   )
   &Color(
    Color=&ColorFromCurve(ColorCurve=$User.HueRamp, CurveIndex=HlslExpression() {hlslExpression="fmod(System.FloatIntegrator,1.0)"})
    Scale Color=<0.5 0.5 0.5>
   )
  ]
  rendererClasses [&NiagaraMeshRendererProperties]
  renderers [{rendererClass=&NiagaraMeshRendererProperties, rendererIndex=0}]
 }
 5_LightningCore {
  emitterUpdateScript [
   &EmitterState()
   &SpawnRate(
    SpawnRate=&FloatFromCurve(
     FloatCurve=DataInterface() {propertyValues {Curve {keys [<at 0.3783 0.05>,<at 0.7507 1.0023>]}}}
     CurveIndex=&FN_FrontEndFloatModifier(Gameplay=$System.NormalizedSystemSpeed, Front End=$User.ParaGlideLeanAlpha)
     Scale Curve=10
    )
    Use Spawn Probability=false
   )
  ]
  particleSpawnScript [
   /SetVariables(Particles.SpawnGroup=HlslExpression() {hlslExpression="int(fmod(float(Particles.UniqueID), 4.0))"}) {isSetParameters=true, module=SetVariables_#1}
   &FN_SkeletalBoneData(Index=4)
   /SetVariables(Particles.DiveDirection:NiagaraFloat=&ClampFloat(Float=&Multiply_Float(
      A=&FN_FrontEndFloatModifier(Gameplay=1.25, Front End=2)
      B=&DotProduct(
       Vector A=&RotateVectorByQuaternion(
        Vector To Rotate=<1 0 0>
        Quaternion=&FN_GetBoneDataRotation(BoneData:ENiagaraBoneData=$Particles.FN_SkeletalBoneData.BoneData, Space:ENiagaraCoordinateSpace=::World)
       )
       Vector B=<0 0 -1>
      )
     ))) {isSetParameters=true}
   &InitializeParticle(
    Color=<rgba 1 0 0 1>
    Lifetime=&UniformRangedFloat(Minimum=0.2, Maximum=0.3)
    Mesh Scale=&Multiply_Vector(
     A=&VectorFromFloat(Value=0.8)
     B=&Multiply_VectorByFloat(
      Vector=&UniformRangedVector(Minimum=<2 4 4>, Maximum=<8 5 5>)
      Float=&Lerp_Float(A=1, B=1.5, Alpha=&Add_Float(A=$Particles.DiveDirection, B=$User.ParaGlideLeanAlpha))
     )
    )
    Mesh Scale Mode=::NewEnumerator4
    UsePositionOffset=false
    Write Material Random=true
   ) {module=InitializeParticle001}
   /SetVariables(Particles.RandomFloat=&UniformRangedFloat(Minimum=0, Maximum=90)) {isSetParameters=true}
  ]
  particleUpdateScript [
   &ParticleState()
   &SampleSkeletalMeshSkeleton(
    Interpolate Bone Sampling=false
    Mesh:NiagaraDataInterfaceSkeletalMesh=$User.SkeletalMesh
    Sampled Position Coordinate Space:ENiagaraCoordinateSpace=::World
    Specific Bone Index:NiagaraInt32=$Particles.SpawnGroup
   ) {enabled=false}
   &SkeletalMeshLocation(Bone / Socket Index=$Particles.SpawnGroup, Bone Sampling Mode=::NewEnumerator4, MaskSpawnGroup=false, Skeletal Mesh=$User.SkeletalMesh)
   &FN_OrientAndSpin(
    Angle=HlslExpression() {hlslExpression=Particles.RandomFloat}
    Axis:Vector3f=&FN_FrontEndVector3Modifier(
     Gameplay=$Particles.SampleSkeletalMeshSkeleton.SampledVelocity
     Front End=&Lerp_Vector(
      A=<0 0 -1>
      B=&TransformVector(Vector=<0 1.2 -1>, Source Space=::World, Destination Space=::Local)
      Alpha=$User.ParaGlideLeanAlpha
     )
    )
   )
   &DynamicMaterialParameters(
    Index 0 Param 2=0.1
    Index 0 Param 3=0.1374
    Index 0 Param 4=0.1
    Module.Param0WriteEnabled=true
    Module.Param1WriteEnabled=true
    Module.Param2WriteEnabled=true
    Module.Param3WriteEnabled=false
   )
   &ScaleMeshSize(Scale Factor=&MakeVector(
     X=&FloatFromCurve(FloatCurve=DataInterface() {propertyValues {Curve {keys [<at -0.0103 0.0997 RCIM_Constant>,<at 0.0926 0.2677 RCIM_Constant>,<at 0.4273 0.4806 RCIM_Constant>,<at 0.6075 0.5703 RCIM_Constant>,<at 0.8349 0.7943 RCIM_Constant>,<at 0.9894 1.0072 RCIM_Constant>]}}})
     Y=&FloatFromCurve(FloatCurve=DataInterface() {propertyValues {Curve {keys [<at -0.0257 0.98 RCIM_Constant>,<at 0.2295 0.59 RCIM_Constant>,<at 0.413 0.3025 RCIM_Constant>,<at 0.6111 0.7242 RCIM_Constant>,<at 0.8093 0.1492 RCIM_Constant>]}}})
     Z=&FloatFromCurve(FloatCurve=DataInterface() {propertyValues {Curve {keys [<at -0.0046 1.0493 RCIM_Constant>,<at 0.1679 0.6027 RCIM_Constant>,<at 0.3751 0.4094 RCIM_Constant>,<at 0.6785 0.6413 RCIM_Constant>,<at 0.9524 0.2355 RCIM_Constant>]}}})
    ))
   &Color(
    Color=&ColorFromCurve(ColorCurve=$User.HueRamp, CurveIndex=HlslExpression() {hlslExpression="fmod(System.FloatIntegrator+.08333,1.0)"})
    Scale Color=<2 2 2>
   )
  ]
  rendererClasses [&NiagaraMeshRendererProperties]
  renderers [{rendererClass=&NiagaraMeshRendererProperties, rendererIndex=0}]
 }
}
summary {
 systemName=NS_Contrail_ReactiveLight
 emitters [
  {emitterName=2_Flare, rendererClasses [&NiagaraSpriteRendererProperties]}
  {emitterName=4_Lightning, rendererClasses [&NiagaraMeshRendererProperties]}
  {emitterName=5_LightningCore, rendererClasses [&NiagaraMeshRendererProperties]}
  {emitterName=3_Glow, rendererClasses [&NiagaraSpriteRendererProperties]}
  {emitterName=4_FN_Sparks, rendererClasses [&NiagaraSpriteRendererProperties]}
 ]
 userVariables [
  User.AccumulatedSpeed {type=NiagaraFloat, defaultValue=0, description=""}
  User.bIsFrontEnd {type=NiagaraBool}
  User.bIsFrontEndPreview {type=NiagaraBool}
  User.HueRamp {
   type=NiagaraDataInterfaceColorCurve
   description=""
   defaultValue {
    struct=NiagaraExt_VariableValue_DataInterface
    value {
     dataInterface=&self:NiagaraDataInterfaceColorCurve_0
     dataInterfaceClass=&NiagaraDataInterfaceColorCurve
    }
   }
  }
  User.ParaGlideLeanAlpha {type=NiagaraFloat, defaultValue=0, description=""}
  User.PlayerSpeed {type=NiagaraFloat, defaultValue=0, description=""}
  User.RibbonIDOffset {type=NiagaraInt32}
  User.SkeletalMesh {
   type=NiagaraDataInterfaceSkeletalMesh
   defaultValue {
    struct=NiagaraExt_VariableValue_DataInterface
    value {dataInterface=&self:NiagaraDataInterfaceSkeletalMesh_0}
   }
  }
  User.VectorOffset {type=Vector3f, description="", defaultValue {struct=Vector3f, value=<0 0 0>}}
 ]
}
